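# Niagara confetti effect for the BasilOkra console in WildEstate, with three sprite burst emitters.
# One launches upward immediately; two offset to opposite sides fire cone bursts after System.BeatDelay, the last slightly later.
# Each packs three spawn groups into one emitter, using VisibilityTag for its three renderers and group-based lerps for drag and gravity.
# Per-sprite colour and size attributes drive randomized shrink and twinkle curves, plus time-of-day tinting.
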NiagaraSystem /WildEstate_BasilOkra/NS_BasilOkra_Console_Confetti
dependencies {
 usedDataInterfaces [&NiagaraDataInterfaceArrayFloat, &NiagaraDataInterfaceCurve, &NiagaraDataInterfaceArrayColor]
 usedDynamicInputs [
  &SelectFloatFromArray
  &SpawnGroup
  &SelectLinearColorFromArray
 ]
 usedModules [
  &SystemState
  &self:SystemSpawnScript.SystemScriptSource.SystemScriptGraph.NiagaraNodeAssignment_0.SetVariables_#1
  &self:SystemSpawnScript.SystemScriptSource.SystemScriptGraph.NiagaraNodeAssignment_1.SetVariables_#2
 ]
}
emitters {
 Confetti_Packed {
  emitterUpdateScript [
   &EmitterState()
   &SpawnBurst_Instantaneous(Spawn Count=80, Use Spawn Probability=false)
   &SpawnBurst_Instantaneous(Spawn Count=5, Spawn Group=1, Use Spawn Probability=false) {module=SpawnBurst_Instantaneous001}
   &SpawnBurst_Instantaneous(Spawn Count=25, Spawn Group=2, Use Spawn Probability=false) {module=SpawnBurst_Instantaneous002}
  ]
  particleSpawnScript [
   /SetVariables(Particles.SpawnGroup=$Emitter.SpawnGroup) {isSetParameters=true, module=SetVariables_#3}
   /SetVariables(Particles.VisibilityTag=$Particles.SpawnGroup) {isSetParameters=true, module=SetVariables_#4}
   &InitializeParticle(
    Color=&RandomRangeLinearColor(Minimum=<rgba 0.9896 1 0.9792 1>, Maximum=<rgba 0.013 0.3569 0.75 1>)
    Lifetime=&RandomRangeFloat(Minimum=2, Maximum=2.5)
    Sprite Rotation Mode=::NewEnumerator2
    Sprite Size Mode=::NewEnumerator1
    UsePositionOffset=false
    Write Material Random=true
   )
   /SetVariables(
    Particles.Sprite_01_Color:LinearColor=&MakeCustomLinearColorFromBool(
     Bool=&RandomBool()
     True Color=&RandomRangeLinearColor(Minimum=<rgba 0 0.305 0 1>, Maximum=<rgba 0.4 0.305 0 1>)
     False Color=&RandomRangeLinearColor(Minimum=<rgba 0.85 0.305 0 1>, Maximum=<rgba 1 0.305 0 1>)
    )
    Particles.Sprite_01_Size:Vector2f=&Vector2DFromFloat(Value=&RandomRangeFloat(Minimum=10, Maximum=15))
   ) {isSetParameters=true, module=SetVariables_#5}
   /SetVariables(
    Particles.Sprite_02_Color:LinearColor=&HueShiftLinearColor(
     Hue=0
     Saturation=&RandomRangeFloat(Minimum=2, Maximum=8)
     Value=&RandomRangeFloat(Minimum=1, Maximum=10)
     Linear Color=&RandomRangeLinearColor(Minimum=<rgba 5 5 5 1>, Maximum=<rgba 10 10 10 1>, Color Channel Mode=::NewEnumerator2)
    )
    Particles.Sprite_02_Size:Vector2f=&Vector2DFromFloat(Value=&RandomRangeFloat(Minimum=5, Maximum=7.5))
   ) {isSetParameters=true, module=SetVariables_#6}
   &ShapeLocation(
    Cylinder Distribution=::NewEnumerator2
    Cylinder Height=0
    Cylinder Radius=250
    MaskSpawnGroup=false
    Offset=<0 100 0>
    Offset Mode=::NewEnumerator0
    Rotation Mode=::NewEnumerator2
    Shape Primitive=::NewEnumerator1
    U Position=&RandomRangeFloat(Minimum=0, Maximum=0.25)
    UseOwnerScale=true
    Yaw / Pitch / Roll=<0.625 0 0>
   )
   &AddVelocity(Velocity=&RandomRangeVector(Minimum=<-200 0 500>, Maximum=<200 250 1000>))
  ]
  particleUpdateScript [
   &ParticleState()
   &Drag(
    Drag=/Niagara/DynamicInputs/Math/LerpMultipleFloats(
     Float Count:ENiagara_MultipleLerpCount=::NewEnumerator0
     1:NiagaraFloat=&RandomRangeFloat(Minimum=3, Maximum=5)
     2:NiagaraFloat=&RandomRangeFloat(Minimum=3, Maximum=5)
     3:NiagaraFloat=&RandomRangeFloat(Minimum=2, Maximum=4)
     4:NiagaraFloat=?
     5:NiagaraFloat=?
     Alpha:NiagaraFloat=&Divide_Float(
      A=&MakeFloatFromInt(Int=3)
      B=&MakeFloatFromInt(Int=$Particles.SpawnGroup)
     )
    )
    Use Linear Drag=true
    Use Rotational Drag=false
   )
   &GravityForce(Gravity=/Niagara/DynamicInputs/Math/LerpMultipleVectors(
     Alpha:NiagaraFloat=&Divide_Float(
      A=&MakeFloatFromInt(Int=3)
      B=&MakeFloatFromInt(Int=$Particles.SpawnGroup)
     )
     V 1:Vector3f=&RandomRangeVector(Minimum=<0 0 -250>, Maximum=<1 1 -100>)
     V 2:Vector3f=&RandomRangeVector(Minimum=<0 0 -250>, Maximum=<1 1 -100>)
     V 3:Vector3f=&RandomRangeVector(Minimum=<0 0 -100>, Maximum=<1 1 -300>)
     V 4:Vector3f=?
     V 5:Vector3f=?
     Vector Count:ENiagara_MultipleLerpCount=::NewEnumerator0
    ))
   &SolveForcesAndVelocity(Clamp Velocity=false, Limit Acceleration=false)
   &ScaleSpriteSize(Scale Factor=&Vector2DFromFloat(Value=&FloatFromCurve(
      FloatCurve=DataInterface() {propertyValues {Curve {keys [
          {arriveTangent=-0.4596, interpMode=RCIM_Cubic, leaveTangent=-0.4596, tangentMode=RCTM_User, time=0.55, value=1}
          {arriveTangent=-4.9031, interpMode=RCIM_Cubic, leaveTangent=-4.9031, tangentMode=RCTM_User, time=1, value=0}
         ]}}}
      CurveIndex=&Divide_Float(
       A=$Particles.NormalizedAge
       B=&RandomRangeFloat(Minimum=0.2, Maximum=1, Randomness Mode=::NewEnumerator2, Random Seed=?)
      )
     )))
   &ScaleColor(
    Scale RGB=&MakeVector(
     X=1
     Y=&Add_Float(
      A=&RandomRangeFloat(Minimum=0, Maximum=1000)
      B=&Lerp_Float(
       A=0
       B=&RandomRangeFloat(Minimum=25, Maximum=150)
       Alpha=&Divide_Float(
        A=$Particles.NormalizedAge
        B=&RandomRangeFloat(Minimum=0.2, Maximum=1, Randomness Mode=::NewEnumerator2, Random Seed=?)
       )
      )
     )
     Z=1
    )
    ScaleA=false
    ScaleRGB=true
   )
   /SetVariables(
    Particles.Sprite_01_Color:LinearColor=&MultiplyLinearColor(Linear Color=$Particles.Initial.Sprite_01_Color, Scale Factor=<1 1 1 1>)
    Particles.Sprite_01_Size:Vector2f=&Multiply_Vector2D(
     A=$Particles.Initial.Sprite_01_Size
     B=&Vector2DFromFloat(Value=&FloatFromCurve(
       FloatCurve=DataInterface() {propertyValues {Curve {keys [
           {arriveTangent=-0.4596, interpMode=RCIM_Cubic, leaveTangent=-0.4596, tangentMode=RCTM_User, time=0.55, value=1}
           {arriveTangent=-4.9031, interpMode=RCIM_Cubic, leaveTangent=-4.9031, tangentMode=RCTM_User, time=1, value=0}
          ]}}}
       CurveIndex=&Divide_Float(
        A=$Particles.NormalizedAge
        B=&RandomRangeFloat(Minimum=0.5, Maximum=1, Randomness Mode=::NewEnumerator2, Random Seed=?)
       )
      ))
    )
   ) {isSetParameters=true, module=SetVariables_#7}
   /SetVariables(
    Particles.Sprite_02_Color:LinearColor=&MultiplyLinearColor(
     Linear Color=$Particles.Initial.Sprite_02_Color
     Scale Factor=&MakeVector4_FromLinearColor(LinearColor=&MultiplyLinearColor(
       Linear Color=&FN_Select_Color_from_Time_Of_Day(
        Use Manual Time of Day Thresholds=true
        Morning=<rgba 1 0.75 0.7753 1>
        Day=<rgba 1 0.8842 0.8294 1>
        Evening=<rgba 1 0.625 0.5 1>
        Night=<rgba 0.7 0.3494 0.4483 1>
        Evening Threshold=21
       )
       Scale Factor=&MakeFloat4FromVectorAndFloat(
        "Vector (RGB)"=<1 1 1>
        "Float (Alpha)"=&Lerp_Float(
         A=1
         B=0
         Alpha=&Divide_Float(
          A=$Particles.NormalizedAge
          B=&RandomRangeFloat(Minimum=0.2, Maximum=1, Randomness Mode=::NewEnumerator2, Random Seed=?)
         )
        )
       )
      ))
    )
    Particles.Sprite_02_Size:Vector2f=&Multiply_Vector2D(
     A=$Particles.Initial.Sprite_02_Size
     B=&Vector2DFromFloat(Value=&FloatFromCurve(
       FloatCurve=DataInterface() {propertyValues {Curve {keys [<at 0.0018 0.1957>,<at 0.0527 0.819>,<at 0.1169 0.6555>,<at 0.1748 0.9015>,<at 0.2394 0.582>,<at 0.332 1.0023>,<at 0.4947 0.3707>,<at 0.6727 1>,<at 1 0>]}}}
       CurveIndex=&Divide_Float(
        A=$Particles.NormalizedAge
        B=&RandomRangeFloat(Minimum=0.2, Maximum=1, Randomness Mode=::NewEnumerator2, Random Seed=?)
       )
      ))
    )
   ) {isSetParameters=true, module=SetVariables_#8}
  ]
  rendererClasses [&NiagaraSpriteRendererProperties]
  renderers [
   {rendererClass=&NiagaraSpriteRendererProperties}
   {rendererClass=&NiagaraSpriteRendererProperties, rendererIndex=1}
   {rendererClass=&NiagaraSpriteRendererProperties, rendererIndex=2}
  ]
 }
 Confetti_Packed002 {
  emitterUpdateScript [
   &EmitterState()
   &SpawnBurst_Instantaneous(Spawn Count=80, Spawn Time=$System.BeatDelay, Use Spawn Probability=false)
   &SpawnBurst_Instantaneous(Spawn Count=5, Spawn Group=1, Spawn Time=$System.BeatDelay, Use Spawn Probability=false) {module=SpawnBurst_Instantaneous001}
   &SpawnBurst_Instantaneous(Spawn Count=25, Spawn Group=2, Spawn Time=$System.BeatDelay, Use Spawn Probability=false) {module=SpawnBurst_Instantaneous002}
  ]
  particleSpawnScript [
   /SetVariables(Particles.SpawnGroup=$Emitter.SpawnGroup) {isSetParameters=true, module=SetVariables_#3}
   /SetVariables(Particles.VisibilityTag=$Particles.SpawnGroup) {isSetParameters=true, module=SetVariables_#4}
   &InitializeParticle(
    Color=&RandomRangeLinearColor(Minimum=<rgba 0.9896 1 0.9792 1>, Maximum=<rgba 0.013 0.3569 0.75 1>)
    Lifetime=&RandomRangeFloat(Minimum=2, Maximum=2.5)
    Position Offset=<-264 -75 220>
    Sprite Rotation Mode=::NewEnumerator2
    Sprite Size Mode=::NewEnumerator1
    UsePositionOffset=true
    Write Material Random=true
   )
   /SetVariables(
    Particles.Sprite_01_Color:LinearColor=&MakeCustomLinearColorFromBool(
     Bool=&RandomBool()
     True Color=&RandomRangeLinearColor(Minimum=<rgba 0 0.305 0 1>, Maximum=<rgba 0.4 0.305 0 1>)
     False Color=&RandomRangeLinearColor(Minimum=<rgba 0.85 0.305 0 1>, Maximum=<rgba 1 0.305 0 1>)
    )
    Particles.Sprite_01_Size:Vector2f=&Vector2DFromFloat(Value=&RandomRangeFloat(Minimum=10, Maximum=15))
   ) {isSetParameters=true, module=SetVariables_#5}
   /SetVariables(
    Particles.Sprite_02_Color:LinearColor=&HueShiftLinearColor(
     Hue=0
     Saturation=&RandomRangeFloat(Minimum=2, Maximum=8)
     Value=&RandomRangeFloat(Minimum=1, Maximum=10)
     Linear Color=&RandomRangeLinearColor(Minimum=<rgba 5 5 5 1>, Maximum=<rgba 10 10 10 1>, Color Channel Mode=::NewEnumerator2)
    )
    Particles.Sprite_02_Size:Vector2f=&Vector2DFromFloat(Value=&RandomRangeFloat(Minimum=5, Maximum=7.5))
   ) {isSetParameters=true, module=SetVariables_#6}
   &ShapeLocation(MaskSpawnGroup=false, Override Seed=false, Sphere Radius=10, UseOwnerScale=true)
   &AddVelocity(
    Module.UseConeFalloff=false
    Module.UseVelDistribution=false
    Override Seed=false
    Rotation Mode=::NewEnumerator2
    Velocity Mode=::NewEnumerator2
    Velocity Speed=&RandomRangeFloat(Minimum=50, Maximum=1000)
    Yaw / Pitch / Roll:Vector3f=<0 28.0243 0>
   )
  ]
  particleUpdateScript [
   &ParticleState()
   &Drag(
    Drag=&LerpMultipleFloats(
     Float Count:ENiagara_MultipleLerpCount=::NewEnumerator0
     1:NiagaraFloat=&RandomRangeFloat(Minimum=3, Maximum=5)
     2:NiagaraFloat=&RandomRangeFloat(Minimum=3, Maximum=5)
     3:NiagaraFloat=&RandomRangeFloat(Minimum=2, Maximum=4)
     4:NiagaraFloat=?
     5:NiagaraFloat=?
     Alpha:NiagaraFloat=&Divide_Float(
      A=&MakeFloatFromInt(Int=3)
      B=&MakeFloatFromInt(Int=$Particles.SpawnGroup)
     )
    )
    Use Linear Drag=true
    Use Rotational Drag=false
   )
   &GravityForce(Gravity=&LerpMultipleVectors(
     Alpha:NiagaraFloat=&Divide_Float(
      A=&MakeFloatFromInt(Int=3)
      B=&MakeFloatFromInt(Int=$Particles.SpawnGroup)
     )
     V 1:Vector3f=&RandomRangeVector(Minimum=<0 0 -250>, Maximum=<1 1 -100>)
     V 2:Vector3f=&RandomRangeVector(Minimum=<0 0 -250>, Maximum=<1 1 -100>)
     V 3:Vector3f=&RandomRangeVector(Minimum=<0 0 -100>, Maximum=<1 1 -300>)
     V 4:Vector3f=?
     V 5:Vector3f=?
     Vector Count:ENiagara_MultipleLerpCount=::NewEnumerator0
    ))
   &SolveForcesAndVelocity(Clamp Velocity=false, Limit Acceleration=false)
   &ScaleSpriteSize(Scale Factor=&Vector2DFromFloat(Value=&FloatFromCurve(
      FloatCurve=DataInterface() {propertyValues {Curve {keys [
          {arriveTangent=-0.4596, interpMode=RCIM_Cubic, leaveTangent=-0.4596, tangentMode=RCTM_User, time=0.55, value=1}
          {arriveTangent=-4.9031, interpMode=RCIM_Cubic, leaveTangent=-4.9031, tangentMode=RCTM_User, time=1, value=0}
         ]}}}
      CurveIndex=&Divide_Float(
       A=$Particles.NormalizedAge
       B=&RandomRangeFloat(Minimum=0.2, Maximum=1, Randomness Mode=::NewEnumerator2, Random Seed=?)
      )
     )))
   &ScaleColor(
    Scale RGB=&MakeVector(
     X=1
     Y=&Add_Float(
      A=&RandomRangeFloat(Minimum=0, Maximum=1000)
      B=&Lerp_Float(
       A=0
       B=&RandomRangeFloat(Minimum=25, Maximum=150)
       Alpha=&Divide_Float(
        A=$Particles.NormalizedAge
        B=&RandomRangeFloat(Minimum=0.2, Maximum=1, Randomness Mode=::NewEnumerator2, Random Seed=?)
       )
      )
     )
     Z=1
    )
    ScaleA=false
    ScaleRGB=true
   )
   /SetVariables(
    Particles.Sprite_01_Color:LinearColor=&MultiplyLinearColor(Linear Color=$Particles.Initial.Sprite_01_Color, Scale Factor=<1 1 1 1>)
    Particles.Sprite_01_Size:Vector2f=&Multiply_Vector2D(
     A=$Particles.Initial.Sprite_01_Size
     B=&Vector2DFromFloat(Value=&FloatFromCurve(
       FloatCurve=DataInterface() {propertyValues {Curve {keys [
           {arriveTangent=-0.4596, interpMode=RCIM_Cubic, leaveTangent=-0.4596, tangentMode=RCTM_User, time=0.55, value=1}
           {arriveTangent=-4.9031, interpMode=RCIM_Cubic, leaveTangent=-4.9031, tangentMode=RCTM_User, time=1, value=0}
          ]}}}
       CurveIndex=&Divide_Float(
        A=$Particles.NormalizedAge
        B=&RandomRangeFloat(Minimum=0.5, Maximum=1, Randomness Mode=::NewEnumerator2, Random Seed=?)
       )
      ))
    )
   ) {isSetParameters=true, module=SetVariables_#7}
   /SetVariables(
    Particles.Sprite_02_Color:LinearColor=&MultiplyLinearColor(
     Linear Color=$Particles.Initial.Sprite_02_Color
     Scale Factor=&MakeVector4_FromLinearColor(LinearColor=&MultiplyLinearColor(
       Linear Color=&FN_Select_Color_from_Time_Of_Day(
        Use Manual Time of Day Thresholds=true
        Morning=<rgba 1 0.75 0.7753 1>
        Day=<rgba 1 0.8842 0.8294 1>
        Evening=<rgba 1 0.625 0.5 1>
        Night=<rgba 0.7 0.3494 0.4483 1>
        Evening Threshold=21
       )
       Scale Factor=&MakeFloat4FromVectorAndFloat(
        "Vector (RGB)"=<1 1 1>
        "Float (Alpha)"=&Lerp_Float(
         A=1
         B=0
         Alpha=&Divide_Float(
          A=$Particles.NormalizedAge
          B=&RandomRangeFloat(Minimum=0.2, Maximum=1, Randomness Mode=::NewEnumerator2, Random Seed=?)
         )
        )
       )
      ))
    )
    Particles.Sprite_02_Size:Vector2f=&Multiply_Vector2D(
     A=$Particles.Initial.Sprite_02_Size
     B=&Vector2DFromFloat(Value=&FloatFromCurve(
       FloatCurve=DataInterface() {propertyValues {Curve {keys [<at 0.0018 0.1957>,<at 0.0527 0.819>,<at 0.1169 0.6555>,<at 0.1748 0.9015>,<at 0.2394 0.582>,<at 0.332 1.0023>,<at 0.4947 0.3707>,<at 0.6727 1>,<at 1 0>]}}}
       CurveIndex=&Divide_Float(
        A=$Particles.NormalizedAge
        B=&RandomRangeFloat(Minimum=0.2, Maximum=1, Randomness Mode=::NewEnumerator2, Random Seed=?)
       )
      ))
    )
   ) {isSetParameters=true, module=SetVariables_#8}
  ]
  rendererClasses [&NiagaraSpriteRendererProperties]
  renderers [
   {rendererClass=&NiagaraSpriteRendererProperties}
   {rendererClass=&NiagaraSpriteRendererProperties, rendererIndex=1}
   {rendererClass=&NiagaraSpriteRendererProperties, rendererIndex=2}
  ]
 }
 Confetti_Packed003 {
  emitterUpdateScript [
   &EmitterState()
   &SpawnBurst_Instantaneous(Spawn Count=80, Spawn Time=&Add_Float(A=0.2, B=$System.BeatDelay), Use Spawn Probability=false)
   &SpawnBurst_Instantaneous(Spawn Count=5, Spawn Group=1, Spawn Time=&Add_Float(A=0.2, B=$System.BeatDelay), Use Spawn Probability=false) {module=SpawnBurst_Instantaneous001}
   &SpawnBurst_Instantaneous(Spawn Count=25, Spawn Group=2, Spawn Time=&Add_Float(A=0.2, B=$System.BeatDelay), Use Spawn Probability=false) {module=SpawnBurst_Instantaneous002}
  ]
  particleSpawnScript [
   /SetVariables(Particles.SpawnGroup=$Emitter.SpawnGroup) {isSetParameters=true, module=SetVariables_#3}
   /SetVariables(Particles.VisibilityTag=$Particles.SpawnGroup) {isSetParameters=true, module=SetVariables_#4}
   &InitializeParticle(
    Color=&RandomRangeLinearColor(Minimum=<rgba 0.9896 1 0.9792 1>, Maximum=<rgba 0.013 0.3569 0.75 1>)
    Lifetime=&RandomRangeFloat(Minimum=2, Maximum=2.5)
    Position Offset=<248 -75 220>
    Sprite Rotation Mode=::NewEnumerator2
    Sprite Size Mode=::NewEnumerator1
    UsePositionOffset=true
    Write Material Random=true
   )
   /SetVariables(
    Particles.Sprite_01_Color:LinearColor=&MakeCustomLinearColorFromBool(
     Bool=&RandomBool()
     True Color=&RandomRangeLinearColor(Minimum=<rgba 0 0.305 0 1>, Maximum=<rgba 0.4 0.305 0 1>)
     False Color=&RandomRangeLinearColor(Minimum=<rgba 0.85 0.305 0 1>, Maximum=<rgba 1 0.305 0 1>)
    )
    Particles.Sprite_01_Size:Vector2f=&Vector2DFromFloat(Value=&RandomRangeFloat(Minimum=10, Maximum=15))
   ) {isSetParameters=true, module=SetVariables_#5}
   /SetVariables(
    Particles.Sprite_02_Color:LinearColor=&HueShiftLinearColor(
     Hue=0
     Saturation=&RandomRangeFloat(Minimum=2, Maximum=8)
     Value=&RandomRangeFloat(Minimum=1, Maximum=10)
     Linear Color=&RandomRangeLinearColor(Minimum=<rgba 5 5 5 1>, Maximum=<rgba 10 10 10 1>, Color Channel Mode=::NewEnumerator2)
    )
    Particles.Sprite_02_Size:Vector2f=&Vector2DFromFloat(Value=&RandomRangeFloat(Minimum=5, Maximum=7.5))
   ) {isSetParameters=true, module=SetVariables_#6}
   &ShapeLocation(MaskSpawnGroup=false, Override Seed=false, Sphere Radius=10, UseOwnerScale=true)
   &AddVelocity(
    Cone Axis=<-1 0 0>
    Module.UseConeFalloff=false
    Module.UseVelDistribution=false
    Override Seed=false
    Rotation Mode=::NewEnumerator2
    Velocity Mode=::NewEnumerator2
    Velocity Speed=&RandomRangeFloat(Minimum=50, Maximum=1000)
   )
  ]
  particleUpdateScript [
   &ParticleState()
   &Drag(
    Drag=&LerpMultipleFloats(
     Float Count:ENiagara_MultipleLerpCount=::NewEnumerator0
     1:NiagaraFloat=&RandomRangeFloat(Minimum=3, Maximum=5)
     2:NiagaraFloat=&RandomRangeFloat(Minimum=3, Maximum=5)
     3:NiagaraFloat=&RandomRangeFloat(Minimum=2, Maximum=4)
     4:NiagaraFloat=?
     5:NiagaraFloat=?
     Alpha:NiagaraFloat=&Divide_Float(
      A=&MakeFloatFromInt(Int=3)
      B=&MakeFloatFromInt(Int=$Particles.SpawnGroup)
     )
    )
    Use Linear Drag=true
    Use Rotational Drag=false
   )
   &GravityForce(Gravity=&LerpMultipleVectors(
     Alpha:NiagaraFloat=&Divide_Float(
      A=&MakeFloatFromInt(Int=3)
      B=&MakeFloatFromInt(Int=$Particles.SpawnGroup)
     )
     V 1:Vector3f=&RandomRangeVector(Minimum=<0 0 -250>, Maximum=<1 1 -100>)
     V 2:Vector3f=&RandomRangeVector(Minimum=<0 0 -250>, Maximum=<1 1 -100>)
     V 3:Vector3f=&RandomRangeVector(Minimum=<0 0 -100>, Maximum=<1 1 -300>)
     V 4:Vector3f=?
     V 5:Vector3f=?
     Vector Count:ENiagara_MultipleLerpCount=::NewEnumerator0
    ))
   &SolveForcesAndVelocity(Clamp Velocity=false, Limit Acceleration=false)
   &ScaleSpriteSize(Scale Factor=&Vector2DFromFloat(Value=&FloatFromCurve(
      FloatCurve=DataInterface() {propertyValues {Curve {keys [
          {arriveTangent=-0.4596, interpMode=RCIM_Cubic, leaveTangent=-0.4596, tangentMode=RCTM_User, time=0.55, value=1}
          {arriveTangent=-4.9031, interpMode=RCIM_Cubic, leaveTangent=-4.9031, tangentMode=RCTM_User, time=1, value=0}
         ]}}}
      CurveIndex=&Divide_Float(
       A=$Particles.NormalizedAge
       B=&RandomRangeFloat(Minimum=0.2, Maximum=1, Randomness Mode=::NewEnumerator2, Random Seed=?)
      )
     )))
   &ScaleColor(
    Scale RGB=&MakeVector(
     X=1
     Y=&Add_Float(
      A=&RandomRangeFloat(Minimum=0, Maximum=1000)
      B=&Lerp_Float(
       A=0
       B=&RandomRangeFloat(Minimum=25, Maximum=150)
       Alpha=&Divide_Float(
        A=$Particles.NormalizedAge
        B=&RandomRangeFloat(Minimum=0.2, Maximum=1, Randomness Mode=::NewEnumerator2, Random Seed=?)
       )
      )
     )
     Z=1
    )
    ScaleA=false
    ScaleRGB=true
   )
   /SetVariables(
    Particles.Sprite_01_Color:LinearColor=&MultiplyLinearColor(Linear Color=$Particles.Initial.Sprite_01_Color, Scale Factor=<1 1 1 1>)
    Particles.Sprite_01_Size:Vector2f=&Multiply_Vector2D(
     A=$Particles.Initial.Sprite_01_Size
     B=&Vector2DFromFloat(Value=&FloatFromCurve(
       FloatCurve=DataInterface() {propertyValues {Curve {keys [
           {arriveTangent=-0.4596, interpMode=RCIM_Cubic, leaveTangent=-0.4596, tangentMode=RCTM_User, time=0.55, value=1}
           {arriveTangent=-4.9031, interpMode=RCIM_Cubic, leaveTangent=-4.9031, tangentMode=RCTM_User, time=1, value=0}
          ]}}}
       CurveIndex=&Divide_Float(
        A=$Particles.NormalizedAge
        B=&RandomRangeFloat(Minimum=0.5, Maximum=1, Randomness Mode=::NewEnumerator2, Random Seed=?)
       )
      ))
    )
   ) {isSetParameters=true, module=SetVariables_#7}
   /SetVariables(
    Particles.Sprite_02_Color:LinearColor=&MultiplyLinearColor(
     Linear Color=$Particles.Initial.Sprite_02_Color
     Scale Factor=&MakeVector4_FromLinearColor(LinearColor=&MultiplyLinearColor(
       Linear Color=&FN_Select_Color_from_Time_Of_Day(
        Use Manual Time of Day Thresholds=true
        Morning=<rgba 1 0.75 0.7753 1>
        Day=<rgba 1 0.8842 0.8294 1>
        Evening=<rgba 1 0.625 0.5 1>
        Night=<rgba 0.7 0.3494 0.4483 1>
        Evening Threshold=21
       )
       Scale Factor=&MakeFloat4FromVectorAndFloat(
        "Vector (RGB)"=<1 1 1>
        "Float (Alpha)"=&Lerp_Float(
         A=1
         B=0
         Alpha=&Divide_Float(
          A=$Particles.NormalizedAge
          B=&RandomRangeFloat(Minimum=0.2, Maximum=1, Randomness Mode=::NewEnumerator2, Random Seed=?)
         )
        )
       )
      ))
    )
    Particles.Sprite_02_Size:Vector2f=&Multiply_Vector2D(
     A=$Particles.Initial.Sprite_02_Size
     B=&Vector2DFromFloat(Value=&FloatFromCurve(
       FloatCurve=DataInterface() {propertyValues {Curve {keys [<at 0.0018 0.1957>,<at 0.0527 0.819>,<at 0.1169 0.6555>,<at 0.1748 0.9015>,<at 0.2394 0.582>,<at 0.332 1.0023>,<at 0.4947 0.3707>,<at 0.6727 1>,<at 1 0>]}}}
       CurveIndex=&Divide_Float(
        A=$Particles.NormalizedAge
        B=&RandomRangeFloat(Minimum=0.2, Maximum=1, Randomness Mode=::NewEnumerator2, Random Seed=?)
       )
      ))
    )
   ) {isSetParameters=true, module=SetVariables_#8}
  ]
  rendererClasses [&NiagaraSpriteRendererProperties]
  renderers [
   {rendererClass=&NiagaraSpriteRendererProperties}
   {rendererClass=&NiagaraSpriteRendererProperties, rendererIndex=1}
   {rendererClass=&NiagaraSpriteRendererProperties, rendererIndex=2}
  ]
 }
}
summary {
 systemName=NS_BasilOkra_Console_Confetti
 emitters [
  {emitterName=Confetti_Packed, rendererClasses [&NiagaraSpriteRendererProperties]}
  {emitterName=Confetti_Packed002, rendererClasses [&NiagaraSpriteRendererProperties]}
  {emitterName=Confetti_Packed003, rendererClasses [&NiagaraSpriteRendererProperties]}
 ]
 userVariables [
  User.ButtonIndex {type=NiagaraInt32}
  User.Color {type=LinearColor, description="", defaultValue {struct=LinearColor, value=<rgba 0.0349 1 0 1>}}
 ]
}
