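# Sound cue for a one-shot electrical spark effect in the BasilOkra Meltdown sound set, classed as in-game SFX.
# A Random node picks among five equally weighted spark wave variants and feeds a Modulator that varies pitch.
# A dedicated sparks attenuation asset handles spatial falloff, and a shared sparks concurrency set, alongside concurrency override values, governs overlapping instances.
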SoundCue /WildEstate_BasilOkra/Assets/Audio/SFX/SC_BasilOkra_Meltdown_ElecSpark_01
root=Modulator_0
nodes [
 {id=Modulator_0, children [Random_0], properties {pitchMin=0.9, volumeMax=1}}
 {id=Random_0, children [WavePlayer_1,WavePlayer_2,WavePlayer_3,WavePlayer_4,WavePlayer_5], properties {weights [1,1,1,1,1]}}
 {id=WavePlayer_1, properties {soundWaveAssetPtr=/WildEstate_BasilOkra/Assets/Audio/SFX/wav/BasilOkra_Meltdown_Elec_OS_05}}
 {id=WavePlayer_2, properties {soundWaveAssetPtr=/WildEstate_BasilOkra/Assets/Audio/SFX/wav/BasilOkra_Meltdown_Elec_OS_04}}
 {id=WavePlayer_3, properties {soundWaveAssetPtr=/WildEstate_BasilOkra/Assets/Audio/SFX/wav/BasilOkra_Meltdown_Elec_OS_03}}
 {id=WavePlayer_4, properties {soundWaveAssetPtr=/WildEstate_BasilOkra/Assets/Audio/SFX/wav/BasilOkra_Meltdown_Elec_OS_02}}
 {id=WavePlayer_5, properties {soundWaveAssetPtr=/WildEstate_BasilOkra/Assets/Audio/SFX/wav/BasilOkra_Meltdown_Elec_OS_01}}
]
properties {
 attenuationSettings=&/ATT_BasilOkra_Meltdown_Sparks
 maxDistance=4500
 soundClassObject=&Fort_SFX_InGame
 volumeMultiplier=1
 concurrencyOverrides {maxCount=4, retriggerTime=0.25, volumeScale=0.5}
 concurrencySet [&/CONC_BasilOkra_Sparks]
}
